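# Control Rig for the F_MED_QuicheLorraineLime_Tap female body, running a shared DeformRig function for corrective deform bones at joints like shoulders, elbows and wrists.
# With Debug Controls On, construction spawns a control per listed bone, and the forward solve copies those control transforms onto the bones before DeformRig.
# A local Thigh Readers function uses spherical pose readers to output directional and crouch weights, with optional debug drawing and printing.
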ControlRigBlueprint /BRCosmetics/Characters/Player/Female/Medium/Bodies/F_MED_QuicheLorraineLime_Tap/Meshes/F_MED_QuicheLorraineLime_Tap_CtrlRig
graphs [
 New Function_ContainedGraph {nodes [
   n1 = RigVMFunctionEntryNode()
   > RigVMVariableNode(Value=/Script/ControlRig.RigUnit_HierarchyGetParent(bDefaultParent=True, Child=@n1.Driver Item) {is=struct RigVMUnitNode}.Parent) {variable=Driver Item Parent}
   To_String = RigVMDispatchNode(Value=@n1.Driver Item.Name) {template="DISPATCH_RigDispatch_ToString(in Value,out Result)"}
   > RigVMVariableNode(Value=@To_String.Result) {variable=Driver Item Name}
   > RigVMFunction_Sequence = RigVMAggregateNode()
   RigVMFunctionReturnNode(
    ExecuteContext=@RigVMFunction_Sequence.F
    Fr Post=RigVMVariableNode() {variable=Result Fr Post}.Value
    Crouch Pre=RigVMVariableNode() {variable=Result Crouch Pre}.Value
    Crouch Post=RigVMVariableNode() {variable=Result Crouch Post}.Value
    Crouch Ot=RigVMVariableNode() {variable=Result Crouch Ot}.Value
    Crouch Pre Ot=RigVMVariableNode() {variable=Result Crouch Pre Ot}.Value
    Crouch Post Ot=RigVMVariableNode() {variable=Result Crouch Post Ot}.Value
    Ba=RigVMVariableNode() {variable=Result Ba}.Value
    Ba Pre=RigVMVariableNode() {variable=Result Ba Pre}.Value
    Ba Post=RigVMVariableNode() {variable=Result Ba Post}.Value
    In=RigVMVariableNode() {variable=Result In}.Value
    In Pre=RigVMVariableNode() {variable=Result In Pre}.Value
    In Post=RigVMVariableNode() {variable=Result In Post}.Value
    Ot=RigVMVariableNode() {variable=Result Ot}.Value
    Ot Pre=RigVMVariableNode() {variable=Result Ot Pre}.Value
    Ot Post=RigVMVariableNode() {variable=Result Ot Post}.Value
    Fr=RigVMVariableNode() {variable=Result Fr}.Value
    Crouch=RigVMVariableNode() {variable=Result Crouch}.Value
    Fr Pre=RigVMVariableNode() {variable=Result Fr Pre}.Value
   )
   n2 = &RigUnit_SphericalPoseReader(
    ActiveRegionScaleFactors=(PositiveWidth=1,NegativeWidth=1,PositiveHeight=1,NegativeHeight=1)
    ActiveRegionSize=0.001
    Debug=(bDrawDebug=false,bDraw2D=False,bDrawLocalAxes=False,DebugScale=25,DebugSegments=20,DebugThickness=0.05)
    DriverAxis=(X=1,Y=0,Z=0)
    FalloffRegionScaleFactors=(PositiveWidth=1,NegativeWidth=1,PositiveHeight=1,NegativeHeight=1)
    FalloffSize=0.5
    FlipHeightScaling=False
    FlipWidthScaling=False
    RotationOffset=(X=-90,Y=0,Z=0)
    Debug.bDrawDebug=RigVMVariableNode() {variable=Fr Viz}.Value
    DriverItem=RigVMVariableNode() {variable=Driver Item}.Value
    OptionalParentItem=RigVMVariableNode() {variable=Driver Item Parent}.Value
    ExecutePin=@RigVMFunction_Sequence.A
   )
   > n3 = RigVMVariableNode(Value=@n2.OutputParam) {variable=Result Fr}
   RigVMFunction_ControlFlowBranch = &RigVMFunction_ControlFlowBranch(ExecuteContext=@n3.ExecuteContext, Condition=RigVMVariableNode() {variable=Fr Viz}.Value)
   n4 = &RigUnit_SphericalPoseReader(
    ActiveRegionScaleFactors=(PositiveWidth=1,NegativeWidth=1,PositiveHeight=1,NegativeHeight=1)
    ActiveRegionSize=0.001
    Debug=(bDrawDebug=false,bDraw2D=False,bDrawLocalAxes=False,DebugScale=25,DebugSegments=20,DebugThickness=0.05)
    DriverAxis=(X=1,Y=0,Z=0)
    FalloffRegionScaleFactors=(PositiveWidth=1,NegativeWidth=1,PositiveHeight=2,NegativeHeight=2)
    FalloffSize=0.25
    FlipHeightScaling=False
    FlipWidthScaling=False
    RotationOffset=(X=-90,Y=0,Z=45)
    ExecutePin=@RigVMFunction_ControlFlowBranch.Completed
    DriverItem=RigVMVariableNode() {variable=Driver Item}.Value
    OptionalParentItem=RigVMVariableNode() {variable=Driver Item Parent}.Value
    Debug.bDrawDebug=RigVMVariableNode() {variable=Fr Pre Viz}.Value
   )
   > n5 = RigVMVariableNode(Value=@n4.OutputParam) {variable=Result Fr Pre}
   n6 = &RigVMFunction_ControlFlowBranch(Condition=RigVMVariableNode() {variable=Fr Pre Viz}.Value, ExecuteContext=@n5.ExecuteContext)
   n7 = &RigUnit_SphericalPoseReader(
    ActiveRegionScaleFactors=(PositiveWidth=1,NegativeWidth=1,PositiveHeight=1,NegativeHeight=1)
    ActiveRegionSize=0.001
    Debug=(bDrawDebug=false,bDraw2D=False,bDrawLocalAxes=False,DebugScale=25,DebugSegments=20,DebugThickness=0.05)
    DriverAxis=(X=1,Y=0,Z=0)
    FalloffRegionScaleFactors=(PositiveWidth=0.5,NegativeWidth=0.5,PositiveHeight=1,NegativeHeight=1)
    FalloffSize=0.5
    FlipHeightScaling=False
    FlipWidthScaling=False
    RotationOffset=(X=-90,Y=0,Z=-45)
    ExecutePin=@n6.Completed
    Debug.bDrawDebug=RigVMVariableNode() {variable=Fr Post Viz}.Value
    DriverItem=RigVMVariableNode() {variable=Driver Item}.Value
    OptionalParentItem=RigVMVariableNode() {variable=Driver Item Parent}.Value
   )
   > n8 = RigVMVariableNode(Value=@n7.OutputParam) {variable=Result Fr Post}
   n9 = &RigVMFunction_ControlFlowBranch(Condition=RigVMVariableNode() {variable=Fr Post Viz}.Value, ExecuteContext=@n8.ExecuteContext)
   n10 = &RigUnit_SphericalPoseReader(
    ActiveRegionScaleFactors=(PositiveWidth=1,NegativeWidth=1,PositiveHeight=1,NegativeHeight=1)
    ActiveRegionSize=0.001
    Debug=(bDrawDebug=true,bDraw2D=False,bDrawLocalAxes=False,DebugScale=25,DebugSegments=20,DebugThickness=0.05)
    DriverAxis=(X=1,Y=0,Z=0)
    FalloffRegionScaleFactors=(PositiveWidth=1,NegativeWidth=1,PositiveHeight=0.125,NegativeHeight=0.125)
    FalloffSize=0.5
    FlipHeightScaling=False
    FlipWidthScaling=False
    RotationOffset=(X=-75,Y=0,Z=0)
    ExecutePin=@n9.Completed
    DriverItem=RigVMVariableNode() {variable=Driver Item}.Value
    Debug.bDrawDebug=RigVMVariableNode() {variable=Crouch Viz}.Value
    OptionalParentItem=RigVMVariableNode() {variable=Driver Item Parent}.Value
   )
   > n11 = RigVMVariableNode(Value=@n10.OutputParam) {variable=Result Crouch}
   n12 = &RigVMFunction_ControlFlowBranch(Condition=RigVMVariableNode() {variable=Crouch Viz}.Value, ExecuteContext=@n11.ExecuteContext)
   n13 = &RigUnit_SphericalPoseReader(
    ActiveRegionScaleFactors=(PositiveWidth=1,NegativeWidth=1,PositiveHeight=1,NegativeHeight=1)
    ActiveRegionSize=0.001
    Debug=(bDrawDebug=true,bDraw2D=False,bDrawLocalAxes=False,DebugScale=25,DebugSegments=20,DebugThickness=0.05)
    DriverAxis=(X=1,Y=0,Z=0)
    FalloffRegionScaleFactors=(PositiveWidth=0.4825,NegativeWidth=0.5,PositiveHeight=0.075,NegativeHeight=0.075)
    FalloffSize=0.5
    FlipHeightScaling=False
    FlipWidthScaling=False
    RotationOffset=(X=-101.3223,Y=-189.762,Z=42.3664)
    ExecutePin=@n12.Completed
    DriverItem=RigVMVariableNode() {variable=Driver Item}.Value
    OptionalParentItem=RigVMVariableNode() {variable=Driver Item Parent}.Value
    Debug.bDrawDebug=RigVMVariableNode() {variable=Crouch Pre Viz}.Value
   )
   > n14 = RigVMVariableNode(Value=@n13.OutputParam) {variable=Result Crouch Pre}
   n15 = &RigVMFunction_ControlFlowBranch(ExecuteContext=@n14.ExecuteContext, Condition=RigVMVariableNode() {variable=Crouch Pre Viz}.Value)
   n16 = &RigUnit_SphericalPoseReader(
    ActiveRegionScaleFactors=(PositiveWidth=1,NegativeWidth=1,PositiveHeight=1,NegativeHeight=1)
    ActiveRegionSize=0.001
    Debug=(bDrawDebug=true,bDraw2D=False,bDrawLocalAxes=False,DebugScale=25,DebugSegments=20,DebugThickness=0.05)
    DriverAxis=(X=1,Y=0,Z=0)
    FalloffRegionScaleFactors=(PositiveWidth=0.5,NegativeWidth=0.5,PositiveHeight=0.075,NegativeHeight=0.075)
    FalloffSize=0.5
    FlipHeightScaling=False
    FlipWidthScaling=False
    RotationOffset=(X=-101.5496,Y=-169.7841,Z=-42.8133)
    ExecutePin=@n15.Completed
    DriverItem=RigVMVariableNode() {variable=Driver Item}.Value
    OptionalParentItem=RigVMVariableNode() {variable=Driver Item Parent}.Value
    Debug.bDrawDebug=RigVMVariableNode() {variable=Crouch Post Viz}.Value
   )
   > n17 = RigVMVariableNode(Value=@n16.OutputParam) {variable=Result Crouch Post}
   n18 = &RigVMFunction_ControlFlowBranch(ExecuteContext=@n17.ExecuteContext, Condition=RigVMVariableNode() {variable=Crouch Post Viz}.Value)
   n19 = &RigUnit_SphericalPoseReader(
    ActiveRegionScaleFactors=(PositiveWidth=1,NegativeWidth=1,PositiveHeight=1,NegativeHeight=1)
    ActiveRegionSize=0.001
    Debug=(bDrawDebug=true,bDraw2D=False,bDrawLocalAxes=False,DebugScale=25,DebugSegments=20,DebugThickness=0.05)
    DriverAxis=(X=1,Y=0,Z=0)
    FalloffRegionScaleFactors=(PositiveWidth=1,NegativeWidth=1,PositiveHeight=0.666,NegativeHeight=0.333)
    FalloffSize=0.5
    FlipHeightScaling=False
    FlipWidthScaling=False
    RotationOffset=(X=-60,Y=0,Z=0)
    ExecutePin=@n18.Completed
    DriverItem=RigVMVariableNode() {variable=Driver Item}.Value
    OptionalParentItem=RigVMVariableNode() {variable=Driver Item Parent}.Value
    Debug.bDrawDebug=RigVMVariableNode() {variable=Crouch Ot Viz}.Value
   )
   > n20 = RigVMVariableNode(Value=@n19.OutputParam) {variable=Result Crouch Ot}
   n21 = &RigVMFunction_ControlFlowBranch(Condition=RigVMVariableNode() {variable=Crouch Ot Viz}.Value, ExecuteContext=@n20.ExecuteContext)
   n22 = &RigUnit_SphericalPoseReader(
    ActiveRegionScaleFactors=(PositiveWidth=1,NegativeWidth=1,PositiveHeight=1,NegativeHeight=1)
    ActiveRegionSize=0.001
    Debug=(bDrawDebug=true,bDraw2D=False,bDrawLocalAxes=False,DebugScale=25,DebugSegments=20,DebugThickness=0.05)
    DriverAxis=(X=1,Y=0,Z=0)
    FalloffRegionScaleFactors=(PositiveWidth=0.5,NegativeWidth=0.5,PositiveHeight=0.2,NegativeHeight=0.4)
    FalloffSize=0.5
    FlipHeightScaling=False
    FlipWidthScaling=False
    RotationOffset=(X=-110.6339,Y=-201.1783,Z=39.9802)
    ExecutePin=@n21.Completed
    DriverItem=RigVMVariableNode() {variable=Driver Item}.Value
    OptionalParentItem=RigVMVariableNode() {variable=Driver Item Parent}.Value
    Debug.bDrawDebug=RigVMVariableNode() {variable=Crouch Ot Pre Viz}.Value
   )
   > n23 = RigVMVariableNode(Value=@n22.OutputParam) {variable=Result Crouch Pre Ot}
   n24 = &RigVMFunction_ControlFlowBranch(Condition=RigVMVariableNode() {variable=Crouch Ot Pre Viz}.Value, ExecuteContext=@n23.ExecuteContext)
   n25 = &RigUnit_SphericalPoseReader(
    ActiveRegionScaleFactors=(PositiveWidth=1,NegativeWidth=1,PositiveHeight=1,NegativeHeight=1)
    ActiveRegionSize=0.001
    Debug=(bDrawDebug=true,bDraw2D=False,bDrawLocalAxes=False,DebugScale=25,DebugSegments=20,DebugThickness=0.05)
    DriverAxis=(X=1,Y=0,Z=0)
    FalloffRegionScaleFactors=(PositiveWidth=0.5,NegativeWidth=0.5,PositiveHeight=0.2,NegativeHeight=0.4)
    FalloffSize=0.5
    FlipHeightScaling=False
    FlipWidthScaling=False
    RotationOffset=(X=-112.4531,Y=-160.3786,Z=-41.3069)
    ExecutePin=@n24.Completed
    DriverItem=RigVMVariableNode() {variable=Driver Item}.Value
    OptionalParentItem=RigVMVariableNode() {variable=Driver Item Parent}.Value
    Debug.bDrawDebug=RigVMVariableNode() {variable=Crouch Ot Post Viz}.Value
   )
   > n26 = RigVMVariableNode(Value=@n25.OutputParam) {variable=Result Crouch Post Ot}
   n27 = &RigVMFunction_ControlFlowBranch(ExecuteContext=@n26.ExecuteContext, Condition=RigVMVariableNode() {variable=Crouch Ot Post Viz}.Value)
   RigVMDispatchNode(
    Enabled=True
    ScreenColor=(R=1,G=1,B=1,A=1)
    ScreenDuration=0.05
    Prefix=/Script/RigVM.RigVMFunction_StringConcat(B=" - CROUCH OT POST:   ", A=RigVMVariableNode() {variable=Driver Item Name}.Value) {is=struct RigVMUnitNode, template="Concat::Execute(in A,in B,out Result)"}.Result
    Value=RigVMVariableNode() {variable=Result Crouch Post Ot}.Value
    ExecuteContext=@n27.True
   ) {template="DISPATCH_RigVMDispatch_Print(in Prefix,in Value,in Enabled,in ScreenDuration,in ScreenColor)"}
   RigVMDispatchNode(
    Enabled=True
    ScreenColor=(R=1,G=1,B=1,A=1)
    ScreenDuration=0.05
    Prefix=&RigVMFunction_StringConcat(B=" - CROUCH OT PRE:   ", A=RigVMVariableNode() {variable=Driver Item Name}.Value) {template="Concat::Execute(in A,in B,out Result)"}.Result
    Value=RigVMVariableNode() {variable=Result Crouch Pre Ot}.Value
    ExecuteContext=@n24.True
   ) {template="DISPATCH_RigVMDispatch_Print(in Prefix,in Value,in Enabled,in ScreenDuration,in ScreenColor)"}
   RigVMDispatchNode(
    Enabled=True
    ScreenColor=(R=1,G=1,B=1,A=1)
    ScreenDuration=0.05
    Prefix=&RigVMFunction_StringConcat(B=" - Crouch Ot:   ", A=RigVMVariableNode() {variable=Driver Item Name}.Value) {template="Concat::Execute(in A,in B,out Result)"}.Result
    Value=RigVMVariableNode() {variable=Result Crouch Ot}.Value
    ExecuteContext=@n21.True
   ) {template="DISPATCH_RigVMDispatch_Print(in Prefix,in Value,in Enabled,in ScreenDuration,in ScreenColor)"}
   RigVMDispatchNode(
    Enabled=True
    ScreenColor=(R=1,G=1,B=1,A=1)
    ScreenDuration=0.05
    Prefix=&RigVMFunction_StringConcat(B=" - CROUCH POST:   ", A=RigVMVariableNode() {variable=Driver Item Name}.Value) {template="Concat::Execute(in A,in B,out Result)"}.Result
    Value=RigVMVariableNode() {variable=Result Crouch Post}.Value
    ExecuteContext=@n18.True
   ) {template="DISPATCH_RigVMDispatch_Print(in Prefix,in Value,in Enabled,in ScreenDuration,in ScreenColor)"}
   RigVMDispatchNode(
    Enabled=True
    ScreenColor=(R=1,G=1,B=1,A=1)
    ScreenDuration=0.05
    Prefix=&RigVMFunction_StringConcat(B=" - CROUCH PRE:   ", A=RigVMVariableNode() {variable=Driver Item Name}.Value) {template="Concat::Execute(in A,in B,out Result)"}.Result
    Value=RigVMVariableNode() {variable=Result Crouch Pre}.Value
    ExecuteContext=@n15.True
   ) {template="DISPATCH_RigVMDispatch_Print(in Prefix,in Value,in Enabled,in ScreenDuration,in ScreenColor)"}
   RigVMDispatchNode(
    Enabled=True
    ScreenColor=(R=1,G=1,B=1,A=1)
    ScreenDuration=0.05
    Prefix=&RigVMFunction_StringConcat(B=" - Crouch:   ", A=RigVMVariableNode() {variable=Driver Item Name}.Value) {template="Concat::Execute(in A,in B,out Result)"}.Result
    Value=RigVMVariableNode() {variable=Result Crouch}.Value
    ExecuteContext=@n12.True
   ) {template="DISPATCH_RigVMDispatch_Print(in Prefix,in Value,in Enabled,in ScreenDuration,in ScreenColor)"}
   RigVMDispatchNode(
    Enabled=True
    ScreenColor=(R=1,G=1,B=1,A=1)
    ScreenDuration=0.05
    Prefix=&RigVMFunction_StringConcat(B=" - FR POST:   ", A=RigVMVariableNode() {variable=Driver Item Name}.Value) {template="Concat::Execute(in A,in B,out Result)"}.Result
    Value=RigVMVariableNode() {variable=Result Fr Post}.Value
    ExecuteContext=@n9.True
   ) {template="DISPATCH_RigVMDispatch_Print(in Prefix,in Value,in Enabled,in ScreenDuration,in ScreenColor)"}
   RigVMDispatchNode(
    Enabled=True
    ScreenColor=(R=1,G=1,B=1,A=1)
    ScreenDuration=0.05
    Prefix=&RigVMFunction_StringConcat(B=" - FR PRE:   ", A=RigVMVariableNode() {variable=Driver Item Name}.Value) {template="Concat::Execute(in A,in B,out Result)"}.Result
    Value=RigVMVariableNode() {variable=Result Fr Pre}.Value
    ExecuteContext=@n6.True
   ) {template="DISPATCH_RigVMDispatch_Print(in Prefix,in Value,in Enabled,in ScreenDuration,in ScreenColor)"}
   Concat = &RigVMFunction_StringConcat(B=" - FR:   ", A=RigVMVariableNode() {variable=Driver Item Name}.Value) {template="Concat::Execute(in A,in B,out Result)"}
   RigVMDispatchNode(
    Enabled=True
    ScreenColor=(R=1,G=1,B=1,A=1)
    ScreenDuration=0.05
    Prefix=@Concat.Result
    Value=RigVMVariableNode() {variable=Result Fr}.Value
    ExecuteContext=@RigVMFunction_ControlFlowBranch.True
   ) {template="DISPATCH_RigVMDispatch_Print(in Prefix,in Value,in Enabled,in ScreenDuration,in ScreenColor)"}
   n28 = &RigUnit_SphericalPoseReader(
    ActiveRegionScaleFactors=(PositiveWidth=1,NegativeWidth=1,PositiveHeight=1,NegativeHeight=1)
    ActiveRegionSize=0.001
    Debug=(bDrawDebug=false,bDraw2D=False,bDrawLocalAxes=False,DebugScale=25,DebugSegments=20,DebugThickness=0.05)
    DriverAxis=(X=1,Y=0,Z=0)
    ExecutePin=()
    FalloffRegionScaleFactors=(PositiveWidth=1,NegativeWidth=1,PositiveHeight=1,NegativeHeight=1)
    FalloffSize=0.5
    FlipHeightScaling=False
    FlipWidthScaling=False
    RotationOffset=(X=180,Y=0,Z=0)
    DriverItem=RigVMVariableNode() {variable=Driver Item}.Value
    OptionalParentItem=RigVMVariableNode() {variable=Driver Item Parent}.Value
    Debug.bDrawDebug=RigVMVariableNode() {variable=In Viz}.Value
   )
   > n29 = RigVMVariableNode(Value=@n28.OutputParam) {variable=Result In}
   n30 = &RigVMFunction_ControlFlowBranch(ExecuteContext=@n29.ExecuteContext, Condition=RigVMVariableNode() {variable=In Viz}.Value)
   n31 = &RigUnit_SphericalPoseReader(
    ActiveRegionScaleFactors=(PositiveWidth=1,NegativeWidth=1,PositiveHeight=1,NegativeHeight=1)
    ActiveRegionSize=0.001
    Debug=(bDrawDebug=false,bDraw2D=False,bDrawLocalAxes=False,DebugScale=25,DebugSegments=20,DebugThickness=0.05)
    DriverAxis=(X=1,Y=0,Z=0)
    FalloffRegionScaleFactors=(PositiveWidth=2,NegativeWidth=2,PositiveHeight=1,NegativeHeight=1)
    FalloffSize=0.25
    FlipHeightScaling=False
    FlipWidthScaling=False
    RotationOffset=(X=225,Y=0,Z=90)
    ExecutePin=@n30.Completed
    DriverItem=RigVMVariableNode() {variable=Driver Item}.Value
    OptionalParentItem=RigVMVariableNode() {variable=Driver Item Parent}.Value
    Debug.bDrawDebug=RigVMVariableNode() {variable=In Pre Viz}.Value
   )
   > n32 = RigVMVariableNode(Value=@n31.OutputParam) {variable=Result In Pre}
   n33 = &RigVMFunction_ControlFlowBranch(ExecuteContext=@n32.ExecuteContext, Condition=RigVMVariableNode() {variable=In Pre Viz}.Value)
   n34 = &RigUnit_SphericalPoseReader(
    ActiveRegionScaleFactors=(PositiveWidth=1,NegativeWidth=1,PositiveHeight=1,NegativeHeight=1)
    ActiveRegionSize=0.001
    Debug=(bDrawDebug=false,bDraw2D=False,bDrawLocalAxes=False,DebugScale=25,DebugSegments=20,DebugThickness=0.05)
    DriverAxis=(X=1,Y=0,Z=0)
    FalloffRegionScaleFactors=(PositiveWidth=2,NegativeWidth=2,PositiveHeight=1,NegativeHeight=1)
    FalloffSize=0.25
    FlipHeightScaling=False
    FlipWidthScaling=False
    RotationOffset=(X=135,Y=0,Z=90)
    ExecutePin=@n33.Completed
    DriverItem=RigVMVariableNode() {variable=Driver Item}.Value
    OptionalParentItem=RigVMVariableNode() {variable=Driver Item Parent}.Value
    Debug.bDrawDebug=RigVMVariableNode() {variable=In Post Viz}.Value
   )
   > n35 = RigVMVariableNode(Value=@n34.OutputParam) {variable=Result In Post}
   n36 = &RigVMFunction_ControlFlowBranch(ExecuteContext=@n35.ExecuteContext, Condition=RigVMVariableNode() {variable=In Post Viz}.Value)
   RigVMDispatchNode(
    Enabled=True
    ScreenColor=(R=1,G=1,B=1,A=1)
    ScreenDuration=0.05
    Prefix=&RigVMFunction_StringConcat(B=" - IN POST:   ", A=RigVMVariableNode() {variable=Driver Item Name}.Value) {template="Concat::Execute(in A,in B,out Result)"}.Result
    ExecuteContext=@n36.True
    Value=RigVMVariableNode() {variable=Result In Post}.Value
   ) {template="DISPATCH_RigVMDispatch_Print(in Prefix,in Value,in Enabled,in ScreenDuration,in ScreenColor)"}
   RigVMDispatchNode(
    Enabled=True
    ScreenColor=(R=1,G=1,B=1,A=1)
    ScreenDuration=0.05
    Prefix=&RigVMFunction_StringConcat(B=" - IN PRE:   ", A=RigVMVariableNode() {variable=Driver Item Name}.Value) {template="Concat::Execute(in A,in B,out Result)"}.Result
    ExecuteContext=@n33.True
    Value=RigVMVariableNode() {variable=Result In Pre}.Value
   ) {template="DISPATCH_RigVMDispatch_Print(in Prefix,in Value,in Enabled,in ScreenDuration,in ScreenColor)"}
   RigVMDispatchNode(
    Enabled=True
    ScreenColor=(R=1,G=1,B=1,A=1)
    ScreenDuration=0.05
    Prefix=&RigVMFunction_StringConcat(B=" - IN:   ", A=RigVMVariableNode() {variable=Driver Item Name}.Value) {template="Concat::Execute(in A,in B,out Result)"}.Result
    ExecuteContext=@n30.True
    Value=RigVMVariableNode() {variable=Result In}.Value
   ) {template="DISPATCH_RigVMDispatch_Print(in Prefix,in Value,in Enabled,in ScreenDuration,in ScreenColor)"}
   n37 = &RigUnit_SphericalPoseReader(
    ActiveRegionScaleFactors=(PositiveWidth=1,NegativeWidth=1,PositiveHeight=1,NegativeHeight=1)
    ActiveRegionSize=0.001
    Debug=(bDrawDebug=false,bDraw2D=False,bDrawLocalAxes=False,DebugScale=25,DebugSegments=20,DebugThickness=0.05)
    DriverAxis=(X=1,Y=0,Z=0)
    ExecutePin=()
    FalloffRegionScaleFactors=(PositiveWidth=1,NegativeWidth=1,PositiveHeight=1,NegativeHeight=1)
    FalloffSize=0.5
    FlipHeightScaling=False
    FlipWidthScaling=False
    RotationOffset=(X=0,Y=0,Z=0)
    DriverItem=RigVMVariableNode() {variable=Driver Item}.Value
    OptionalParentItem=RigVMVariableNode() {variable=Driver Item Parent}.Value
    Debug.bDrawDebug=RigVMVariableNode() {variable=Ot Viz}.Value
   )
   > n38 = RigVMVariableNode(Value=@n37.OutputParam) {variable=Result Ot}
   n39 = &RigVMFunction_ControlFlowBranch(ExecuteContext=@n38.ExecuteContext, Condition=RigVMVariableNode() {variable=Ot Viz}.Value)
   n40 = &RigUnit_SphericalPoseReader(
    ActiveRegionScaleFactors=(PositiveWidth=1,NegativeWidth=1,PositiveHeight=1,NegativeHeight=1)
    ActiveRegionSize=0.001
    Debug=(bDrawDebug=false,bDraw2D=False,bDrawLocalAxes=False,DebugScale=25,DebugSegments=20,DebugThickness=0.05)
    DriverAxis=(X=1,Y=0,Z=0)
    FalloffRegionScaleFactors=(PositiveWidth=1,NegativeWidth=1,PositiveHeight=2,NegativeHeight=2)
    FalloffSize=0.25
    FlipHeightScaling=False
    FlipWidthScaling=False
    RotationOffset=(X=0,Y=-45,Z=0)
    ExecutePin=@n39.Completed
    Debug.bDrawDebug=RigVMVariableNode() {variable=Ot Pre Viz}.Value
    DriverItem=RigVMVariableNode() {variable=Driver Item}.Value
    OptionalParentItem=RigVMVariableNode() {variable=Driver Item Parent}.Value
   )
   > n41 = RigVMVariableNode(Value=@n40.OutputParam) {variable=Result Ot Pre}
   n42 = &RigVMFunction_ControlFlowBranch(Condition=RigVMVariableNode() {variable=Ot Pre Viz}.Value, ExecuteContext=@n41.ExecuteContext)
   n43 = &RigUnit_SphericalPoseReader(
    ActiveRegionScaleFactors=(PositiveWidth=1,NegativeWidth=1,PositiveHeight=1,NegativeHeight=1)
    ActiveRegionSize=0.001
    Debug=(bDrawDebug=false,bDraw2D=False,bDrawLocalAxes=False,DebugScale=25,DebugSegments=20,DebugThickness=0.05)
    DriverAxis=(X=1,Y=0,Z=0)
    FalloffRegionScaleFactors=(PositiveWidth=1,NegativeWidth=1,PositiveHeight=2,NegativeHeight=2)
    FalloffSize=0.25
    FlipHeightScaling=False
    FlipWidthScaling=False
    RotationOffset=(X=0,Y=45,Z=0)
    ExecutePin=@n42.Completed
    Debug.bDrawDebug=RigVMVariableNode() {variable=Ot Post Viz}.Value
    DriverItem=RigVMVariableNode() {variable=Driver Item}.Value
    OptionalParentItem=RigVMVariableNode() {variable=Driver Item Parent}.Value
   )
   > n44 = RigVMVariableNode(Value=@n43.OutputParam) {variable=Result Ot Post}
   n45 = &RigVMFunction_ControlFlowBranch(Condition=RigVMVariableNode() {variable=Ot Post Viz}.Value, ExecuteContext=@n44.ExecuteContext)
   RigVMDispatchNode(
    Enabled=True
    ScreenColor=(R=1,G=1,B=1,A=1)
    ScreenDuration=0.05
    Prefix=&RigVMFunction_StringConcat(B=" - OT POST:   ", A=RigVMVariableNode() {variable=Driver Item Name}.Value) {template="Concat::Execute(in A,in B,out Result)"}.Result
    ExecuteContext=@n45.True
    Value=RigVMVariableNode() {variable=Result Ot Post}.Value
   ) {template="DISPATCH_RigVMDispatch_Print(in Prefix,in Value,in Enabled,in ScreenDuration,in ScreenColor)"}
   RigVMDispatchNode(
    Enabled=True
    ScreenColor=(R=1,G=1,B=1,A=1)
    ScreenDuration=0.05
    Prefix=&RigVMFunction_StringConcat(B=" - OT PRE:   ", A=RigVMVariableNode() {variable=Driver Item Name}.Value) {template="Concat::Execute(in A,in B,out Result)"}.Result
    ExecuteContext=@n42.True
    Value=RigVMVariableNode() {variable=Result Ot Pre}.Value
   ) {template="DISPATCH_RigVMDispatch_Print(in Prefix,in Value,in Enabled,in ScreenDuration,in ScreenColor)"}
   RigVMDispatchNode(
    Enabled=True
    ScreenColor=(R=1,G=1,B=1,A=1)
    ScreenDuration=0.05
    Prefix=&RigVMFunction_StringConcat(B=" - OT:   ", A=RigVMVariableNode() {variable=Driver Item Name}.Value) {template="Concat::Execute(in A,in B,out Result)"}.Result
    ExecuteContext=@n39.True
    Value=RigVMVariableNode() {variable=Result Ot}.Value
   ) {template="DISPATCH_RigVMDispatch_Print(in Prefix,in Value,in Enabled,in ScreenDuration,in ScreenColor)"}
   n46 = &RigUnit_SphericalPoseReader(
    ActiveRegionScaleFactors=(PositiveWidth=1,NegativeWidth=1,PositiveHeight=1,NegativeHeight=1)
    ActiveRegionSize=0.001
    Debug=(bDrawDebug=false,bDraw2D=False,bDrawLocalAxes=False,DebugScale=25,DebugSegments=20,DebugThickness=0.05)
    DriverAxis=(X=1,Y=0,Z=0)
    ExecutePin=()
    FalloffRegionScaleFactors=(PositiveWidth=1,NegativeWidth=1,PositiveHeight=1,NegativeHeight=1)
    FalloffSize=0.5
    FlipHeightScaling=False
    FlipWidthScaling=False
    RotationOffset=(X=90,Y=0,Z=0)
    DriverItem=RigVMVariableNode() {variable=Driver Item}.Value
    OptionalParentItem=RigVMVariableNode() {variable=Driver Item Parent}.Value
    Debug.bDrawDebug=RigVMVariableNode() {variable=Ba Viz}.Value
   )
   > n47 = RigVMVariableNode(Value=@n46.OutputParam) {variable=Result Ba}
   n48 = &RigVMFunction_ControlFlowBranch(ExecuteContext=@n47.ExecuteContext, Condition=RigVMVariableNode() {variable=Ba Viz}.Value)
   n49 = &RigUnit_SphericalPoseReader(
    ActiveRegionScaleFactors=(PositiveWidth=1,NegativeWidth=1,PositiveHeight=1,NegativeHeight=1)
    ActiveRegionSize=0.001
    Debug=(bDrawDebug=false,bDraw2D=False,bDrawLocalAxes=False,DebugScale=25,DebugSegments=20,DebugThickness=0.05)
    DriverAxis=(X=1,Y=0,Z=0)
    FalloffRegionScaleFactors=(PositiveWidth=1,NegativeWidth=1,PositiveHeight=2,NegativeHeight=2)
    FalloffSize=0.25
    FlipHeightScaling=False
    FlipWidthScaling=False
    RotationOffset=(X=90,Y=0,Z=-45)
    ExecutePin=@n48.Completed
    DriverItem=RigVMVariableNode() {variable=Driver Item}.Value
    OptionalParentItem=RigVMVariableNode() {variable=Driver Item Parent}.Value
    Debug.bDrawDebug=RigVMVariableNode() {variable=Ba Pre Viz}.Value
   )
   > n50 = RigVMVariableNode(Value=@n49.OutputParam) {variable=Result Ba Pre}
   n51 = &RigVMFunction_ControlFlowBranch(ExecuteContext=@n50.ExecuteContext, Condition=RigVMVariableNode() {variable=Ba Pre Viz}.Value)
   n52 = &RigUnit_SphericalPoseReader(
    ActiveRegionScaleFactors=(PositiveWidth=1,NegativeWidth=1,PositiveHeight=1,NegativeHeight=1)
    ActiveRegionSize=0.001
    Debug=(bDrawDebug=false,bDraw2D=False,bDrawLocalAxes=False,DebugScale=25,DebugSegments=20,DebugThickness=0.05)
    DriverAxis=(X=1,Y=0,Z=0)
    FalloffRegionScaleFactors=(PositiveWidth=1,NegativeWidth=1,PositiveHeight=2,NegativeHeight=2)
    FalloffSize=0.25
    FlipHeightScaling=False
    FlipWidthScaling=False
    RotationOffset=(X=90,Y=0,Z=45)
    ExecutePin=@n51.Completed
    DriverItem=RigVMVariableNode() {variable=Driver Item}.Value
    OptionalParentItem=RigVMVariableNode() {variable=Driver Item Parent}.Value
    Debug.bDrawDebug=RigVMVariableNode() {variable=Ba Post Viz}.Value
   )
   > n53 = RigVMVariableNode(Value=@n52.OutputParam) {variable=Result Ba Post}
   n54 = &RigVMFunction_ControlFlowBranch(ExecuteContext=@n53.ExecuteContext, Condition=RigVMVariableNode() {variable=Ba Post Viz}.Value)
   RigVMDispatchNode(
    Enabled=True
    ScreenColor=(R=1,G=1,B=1,A=1)
    ScreenDuration=0.05
    Prefix=&RigVMFunction_StringConcat(B=" - BA POST:   ", A=RigVMVariableNode() {variable=Driver Item Name}.Value) {template="Concat::Execute(in A,in B,out Result)"}.Result
    ExecuteContext=@n54.True
    Value=RigVMVariableNode() {variable=Result Ba Post}.Value
   ) {template="DISPATCH_RigVMDispatch_Print(in Prefix,in Value,in Enabled,in ScreenDuration,in ScreenColor)"}
   RigVMDispatchNode(
    Enabled=True
    ScreenColor=(R=1,G=1,B=1,A=1)
    ScreenDuration=0.05
    Prefix=&RigVMFunction_StringConcat(B=" - BA PRE:   ", A=RigVMVariableNode() {variable=Driver Item Name}.Value) {template="Concat::Execute(in A,in B,out Result)"}.Result
    ExecuteContext=@n51.True
    Value=RigVMVariableNode() {variable=Result Ba Pre}.Value
   ) {template="DISPATCH_RigVMDispatch_Print(in Prefix,in Value,in Enabled,in ScreenDuration,in ScreenColor)"}
   RigVMDispatchNode(
    Enabled=True
    ScreenColor=(R=1,G=1,B=1,A=1)
    ScreenDuration=0.05
    Prefix=&RigVMFunction_StringConcat(B=" - BA:   ", A=RigVMVariableNode() {variable=Driver Item Name}.Value) {template="Concat::Execute(in A,in B,out Result)"}.Result
    ExecuteContext=@n48.True
    Value=RigVMVariableNode() {variable=Result Ba}.Value
   ) {template="DISPATCH_RigVMDispatch_Print(in Prefix,in Value,in Enabled,in ScreenDuration,in ScreenColor)"}
   RigVMRerouteNode(Value=@RigVMFunction_Sequence.C)
   RigVMRerouteNode(Value=@RigVMFunction_Sequence.D)
   RigVMRerouteNode(Value=@RigVMFunction_Sequence.B)
  ]}
 RigVMFunctionLibrary {nodes [Thigh Readers = RigVMCollapseNode(Driver Item=(Type=None,Name="None"), ExecuteContext=())]}
 RigVMFunction_Sequence_1_ContainedGraph {nodes [
   n1 = RigVMFunctionEntryNode()
   RigVMFunction_Sequence = &RigVMFunction_Sequence(ExecuteContext=@n1.ExecuteContext)
   n2 = &RigVMFunction_Sequence(ExecuteContext=@RigVMFunction_Sequence.B)
   n3 = &RigVMFunction_Sequence(ExecuteContext=@n2.B)
   n4 = &RigVMFunction_Sequence(ExecuteContext=@n3.B)
   n5 = &RigVMFunction_Sequence(ExecuteContext=@n4.B)
   RigVMFunctionReturnNode(A=@RigVMFunction_Sequence.A, B=@n2.A, C=@n3.A, D=@n4.A, E=@n5.A, F=@n5.B)
  ]}
 RigVMModel {nodes [
   RigUnit_BeginExecution = &RigUnit_BeginExecution()
   n1 = &RigVMFunction_ControlFlowBranch(ExecuteContext=@RigUnit_BeginExecution.ExecutePin, Condition=RigVMVariableNode() {variable=Debug Controls On}.Value)
   RigVMFunction_ControlFlowBranch = &RigVMFunction_ControlFlowBranch(Condition=RigVMVariableNode() {variable=Debug Controls On}.Value, ExecuteContext=@n1.True)
   For_Each = RigVMDispatchNode(ExecuteContext=@RigVMFunction_ControlFlowBranch.True, Array=RigVMVariableNode() {variable=Debug Ctrls}.Value) {template="DISPATCH_RigVMDispatch_ArrayIterator(in Array,out Element,out Index,out Count,out Ratio)"}
   > &RigUnit_SetTransform(
    Weight=1
    bInitial=False
    Item=RigVMDispatchNode(Array=RigVMVariableNode() {variable=Debug Bones}.Value, Index=@For_Each.Index) {template="DISPATCH_RigVMDispatch_ArrayGetAtIndex(in Array,in Index,out Element)"}.Element
    Value=&RigUnit_GetTransform(bInitial=False, Item=@For_Each.Element).Transform
   )
   FN_DeformRig_FNC = RigVMFunctionReferenceNode(
    DeformRig_ Curve="(deformRig_l_upperarm_up_alpha_72_#1=False,deformRig_r_upperarm_up_alpha_73_#2=False,deformRig_l_upperarm_fr_alpha_74_#3=False,deformRig_r_upperarm_fr_alpha_75_CB51B8 ...(truncated)"
    DeformRig_Bool="(upperarm_1_#4=True,elbow_3_#5=True,wrist_7_#6=True,groin_8_#7=True,glute_17_#8=True,knee_9_40C408324CA261 ...(truncated)"
    DeformRig_Debug=(upperarm_debug_18_#4=False,wrist_debug_20_#6=False,lat_debug_26_#9=False,groinGlute_debug_27_#7=False)
    DeformRig_Value="(upperarm_up_18_#10=(Rotation=(X=0,Y=0,Z=0,W=1),Translation=(X=0,Y=0,Z=8),Scale3D=(X=0.7,Y=1,Z=1)),upperarm_fwd_fr_19_#11=(Rotation=(X=0,Y=0,Z=0,W=1),Translation=(X=0,Y=10,Z=0),Scale3D=(X=0.5,Y=1,Z= ...(truncated)"
    ExecuteContext=@n1.Completed
   )
  ]}
 RigVMModel Rig {nodes [
   n1 = &RigUnit_PrepareForExecution()
   RigVMFunction_Sequence = &RigVMFunction_Sequence(ExecuteContext=@n1.ExecutePin)
   RigVMFunction_ControlFlowBranch = &RigVMFunction_ControlFlowBranch(ExecuteContext=@RigVMFunction_Sequence.A, Condition=RigVMVariableNode() {variable=Debug Controls On}.Value)
   RigVMDispatchNode(ExecuteContext=@RigVMFunction_ControlFlowBranch.True, Array=RigVMVariableNode() {variable=Debug Ctrls}.Value) {template="DISPATCH_RigVMDispatch_ArrayReset(io Array)"}
   ItemArray_1 = RigVMDispatchNode(Value="((Type=Bone,Name=\"calf_l\"),(Type=Bone,Name=\"calf_r\"),(Type=Bone,Name=\"clavicle_l\"),(Type=Bone,Name=\"clavicle_r\"),(Type=Bone,Name=\"upperarm_l\"),(Type=Bone,Name=\"upperarm_r\"),(Type=Bone,Name=\"lowerarm_l\"),(Type=Bone,Name=\"lowerarm_r\"),(Type=Bone,Name=\"thigh_ ...(truncated)") {template="DISPATCH_RigVMDispatch_Constant(io Value)"}
   > RigVMVariableNode(Value=@ItemArray_1.Value) {variable=Debug Bones}
   > For_Each = RigVMDispatchNode(Array=@ItemArray_1.Value) {template="DISPATCH_RigVMDispatch_ArrayIterator(in Array,out Element,out Index,out Count,out Ratio)"}
   Concat = /Script/RigVM.RigVMFunction_NameConcat(B=" Ctrl", A=@For_Each.Element.Name) {is=struct RigVMUnitNode, template="Concat::Execute(in A,in B,out Result)"}
   > n2 = /Script/ControlRig.RigUnit_HierarchyAddControlTransform(
    InitialValue=(Rotation=(X=0,Y=0,Z=0,W=1),Translation=(X=0,Y=0,Z=0),Scale3D=(X=1,Y=1,Z=1))
    OffsetSpace=GlobalSpace
    Settings="(InitialSpace=LocalSpace,bUsePreferredRotationOrder=False,PreferredRotationOrder=YZX,Limits=(LimitTranslationX=(bMinimum=False,bMaximum=False),LimitTranslationY=(bMinimum=False,bMaximum=False),LimitTranslationZ=(bMinimum=False,bMaximum=False),LimitPitch=(b ...(truncated)"
    Name=@Concat.Result
    Parent=&RigUnit_HierarchyGetParent(bDefaultParent=True, Child=@For_Each.Element).Parent
    OffsetTransform=&RigUnit_GetTransform(bInitial=False, Item=@For_Each.Element).Transform
   ) {
    is=struct RigVMUnitNode
    template="SpawnControl::Execute(in InitialValue,in Settings,in OffsetTransform,in OffsetSpace,in Parent,in Name,out Item)"
   }
   > RigVMDispatchNode(Array=RigVMVariableNode() {variable=Debug Ctrls}.Value, Element=@n2.Item) {template="DISPATCH_RigVMDispatch_ArrayAdd(io Array,in Element,out Index)"}
  ]}
]
hierarchy [
 root {loc=<0 0 0>, rot=<0 0 0 1>}
 pelvis {parent=root, loc=<0 -0.5301 98.0571>, rot=<0.0127 0.707 -0.0127 -0.707>}
 spine_01 {parent=pelvis, loc=<6.1208 -0.9328 0>, rot=<0 0 0.034 -0.9994>}
 spine_02 {parent=spine_01, loc=<6.5997 -0.2095 0>, rot=<0 0 0.0238 -0.9997>}
 spine_03 {parent=spine_02, loc=<6.4914 -0.1237 0>, rot=<0 0 0.0149 -0.9999>}
 spine_04 {parent=spine_03, loc=<6.9598 0.0794 0>, rot=<0 0 -0.0132 -0.9999>}
 spine_05 {parent=spine_04, loc=<6.9987 0 0>, rot=<0 0 -0.0015 -1>}
 neck_01 {parent=spine_05, loc=<11.3138 0.1529 0>, rot=<0 0 -0.2003 -0.9797>}
 neck_02 {parent=neck_01, loc=<4.5627 0 0>, rot=<0 0 0.0504 -0.9987>}
 head {parent=neck_02, loc=<5.3914 0 0>, rot=<0 0 0.115 -0.9934>}
 clavicle_l {parent=spine_05, loc=<5.4831 -0.1053 -3.2946>, rot=<0.0465 -0.7289 -0.009 -0.6829>}
 upperarm_l {parent=clavicle_l, loc=<9.8658 -1.2136 -0.4105>, rot=<0.0164 -0.3245 -0.0269 -0.9454>}
 lowerarm_l {parent=upperarm_l, loc=<26.465 -0.8345 0.259>, rot=<0.0036 0.0161 -0.2298 -0.9731>}
 hand_l {parent=lowerarm_l, loc=<22.7355 -0.3203 -0.6105>, rot=<-0.6784 -0.0223 0.0735 -0.7307>}
 thumb_01_l {parent=hand_l, loc=<1.3118 -0.3191 -2.459>, rot=<0.6546 -0.4042 0.0664 -0.6354>}
 thumb_02_l {parent=thumb_01_l, loc=<5.9336 -0.0448 0.0401>, rot=<-0.0037 0.0213 0.298 -0.9543>}
 thumb_03_l {parent=thumb_02_l, loc=<2.2223 0.3682 -0.0907>, rot=<0.0118 -0.013 0.0066 -0.9998>}
 index_metacarpal_l {parent=hand_l, loc=<1.6535 0.0183 -2.96>, rot=<0.0158 -0.0124 -0.0173 -0.9996>}
 index_01_l {parent=index_metacarpal_l, loc=<8.3008 0.0357 -0.0524>, rot=<-0.0044 0.0485 0.0924 -0.9945>}
 index_02_l {parent=index_01_l, loc=<3.3587 -0.3004 0.0452>, rot=<-0.002 0.0027 0.3225 -0.9466>}
 index_03_l {parent=index_02_l, loc=<2.5578 -0.123 -0.0252>, rot=<-0.002 0.0067 0.2558 -0.9667>}
 middle_metacarpal_l {parent=hand_l, loc=<2.1856 -0.1725 -0.5984>, rot=<0.006 0.0168 -0.0329 -0.9993>}
 middle_01_l {parent=middle_metacarpal_l, loc=<7.3275 -0.2447 0.0129>, rot=<-0.0208 0.0042 0.0858 -0.9961>}
 middle_02_l {parent=middle_01_l, loc=<4.0123 0.1401 0.1175>, rot=<-0.0218 0.0258 0.3512 -0.9357>}
 middle_03_l {parent=middle_02_l, loc=<2.6077 0.0534 0.1373>, rot=<0.0082 0.0188 0.1708 -0.9851>}
 ring_metacarpal_l {parent=hand_l, loc=<2.2074 -0.3625 1.3638>, rot=<-0.0284 0.0722 -0.0458 -0.9959>}
 ring_01_l {parent=ring_metacarpal_l, loc=<7.0295 -0.6197 -0.1446>, rot=<-0.0489 -0.0252 0.1892 -0.9804>}
 ring_02_l {parent=ring_01_l, loc=<3.6174 0.4647 -0.1027>, rot=<0 0 0.2806 -0.9598>}
 ring_03_l {parent=ring_02_l, loc=<2.4152 0.114 0.0199>, rot=<0 0 0.1189 -0.9929>}
 pinky_metacarpal_l {parent=hand_l, loc=<1.7616 -0.9753 3.3814>, rot=<-0.0397 0.0814 -0.069 -0.9935>}
 pinky_01_l {parent=pinky_metacarpal_l, loc=<7.1709 -0.4831 -0.1094>, rot=<-0.0654 -0.0035 0.2242 -0.9723>}
 pinky_02_l {parent=pinky_01_l, loc=<2.8274 0.0893 -0.1043>, rot=<0.0037 0.0011 0.2854 -0.9584>}
 pinky_03_l {parent=pinky_02_l, loc=<2.2046 0.0307 -0.0078>, rot=<-0.0002 0.0098 0.1445 -0.9895>}
 deform_wrist_dn_base_l {parent=hand_l, loc=<0 -1.5326 0>, rot=<0 0 0 -1>}
 deform_wrist_dn_l {parent=deform_wrist_dn_base_l, loc=<0 0 0>, rot=<0 0 0 1>}
 deform_wrist_up_base_l {parent=hand_l, loc=<0 2.3009 0>, rot=<0 0 0 -1>}
 deform_wrist_up_l {parent=deform_wrist_up_base_l, loc=<0 0 0>, rot=<0 0 0 1>}
 lowerarm_twist_02_l {parent=lowerarm_l, loc=<10.6707 0 0>, rot=<0 0 0 1>}
 lowerarm_twist_01_l {parent=lowerarm_l, loc=<18.316 0 0>, rot=<0 0 0 1>}
 deform_elbow_in_base_l {parent=lowerarm_l, loc=<0.1834 3.6062 -0.2572>, rot=<0 0 0 -1>}
 deform_elbow_in_l {parent=deform_elbow_in_base_l, loc=<0 0 0>, rot=<0 0 0 1>}
 deform_elbow_out_base_l {parent=lowerarm_l, loc=<0 -4 0>, rot=<0 0 0 1>}
 deform_elbow_out_l {parent=deform_elbow_out_base_l, loc=<0 0 0>, rot=<0 0 0 1>}
 upperarm_twist_01_l {parent=upperarm_l, loc=<7.2135 -0.1134 0>, rot=<0 0 0 1>}
 deform_upperarm_bk_base_l {parent=upperarm_twist_01_l, loc=<-6.4053 -4.3746 0.314>, rot=<0 0 0 -1>}
 deform_upperarm_bk_l {parent=deform_upperarm_bk_base_l, loc=<0 0 0>, rot=<0 0 0 1>}
 deform_upperarm_fr_base_l {parent=upperarm_twist_01_l, loc=<-5.6312 5.092 0.3354>, rot=<0 0 0 -1>}
 deform_upperarm_fr_l {parent=deform_upperarm_fr_base_l, loc=<0 0 0>, rot=<0 0 0 1>}
 deform_upperarm_up_base_l {parent=upperarm_twist_01_l, loc=<-8.891 0.4825 3.9517>, rot=<0 0 0 -1>}
 deform_upperarm_up_l {parent=deform_upperarm_up_base_l, loc=<0 0 0>, rot=<0 0 0 1>}
 upperarm_twist_02_l {parent=upperarm_l, loc=<19.7998 -0.1134 0>, rot=<0 0 0 1>}
 clavicle_r {parent=spine_05, loc=<5.4831 -0.1053 3.2946>, rot=<0.7289 0.0465 -0.6829 0.009>}
 upperarm_r {parent=clavicle_r, loc=<-9.8658 1.2136 0.4105>, rot=<0.0164 -0.3245 -0.0269 -0.9454>}
 lowerarm_r {parent=upperarm_r, loc=<-26.465 0.8345 -0.259>, rot=<0.0036 0.0161 -0.2298 -0.9731>}
 hand_r {parent=lowerarm_r, loc=<-22.7355 0.3203 0.6106>, rot=<-0.6784 -0.0223 0.0735 -0.7307>}
 thumb_01_r {parent=hand_r, loc=<-1.3118 0.3191 2.459>, rot=<0.6546 -0.4042 0.0664 -0.6354>}
 thumb_02_r {parent=thumb_01_r, loc=<-5.9336 0.0448 -0.0401>, rot=<-0.0037 0.0213 0.298 -0.9543>}
 thumb_03_r {parent=thumb_02_r, loc=<-2.2223 -0.3682 0.0907>, rot=<0.0118 -0.013 0.0066 -0.9998>}
 index_metacarpal_r {parent=hand_r, loc=<-1.6535 -0.0183 2.96>, rot=<0.0158 -0.0124 -0.0173 -0.9996>}
 index_01_r {parent=index_metacarpal_r, loc=<-8.3008 -0.0357 0.0524>, rot=<-0.0044 0.0485 0.0924 -0.9945>}
 index_02_r {parent=index_01_r, loc=<-3.3587 0.3004 -0.0452>, rot=<-0.002 0.0027 0.3225 -0.9466>}
 index_03_r {parent=index_02_r, loc=<-2.5578 0.123 0.0252>, rot=<-0.002 0.0067 0.2558 -0.9667>}
 middle_metacarpal_r {parent=hand_r, loc=<-2.1856 0.1725 0.5984>, rot=<0.006 0.0168 -0.0329 -0.9993>}
 middle_01_r {parent=middle_metacarpal_r, loc=<-7.3275 0.2447 -0.0129>, rot=<-0.0208 0.0042 0.0858 -0.9961>}
 middle_02_r {parent=middle_01_r, loc=<-4.0123 -0.1401 -0.1175>, rot=<-0.0218 0.0258 0.3512 -0.9357>}
 middle_03_r {parent=middle_02_r, loc=<-2.6077 -0.0534 -0.1373>, rot=<0.0082 0.0188 0.1708 -0.9851>}
 ring_metacarpal_r {parent=hand_r, loc=<-2.2074 0.3625 -1.3638>, rot=<-0.0284 0.0722 -0.0458 -0.9959>}
 ring_01_r {parent=ring_metacarpal_r, loc=<-7.0295 0.6197 0.1446>, rot=<-0.0489 -0.0252 0.1892 -0.9804>}
 ring_02_r {parent=ring_01_r, loc=<-3.6174 -0.4647 0.1027>, rot=<0 0 0.2806 -0.9598>}
 ring_03_r {parent=ring_02_r, loc=<-2.4152 -0.114 -0.0199>, rot=<0 0 0.1189 -0.9929>}
 pinky_metacarpal_r {parent=hand_r, loc=<-1.7616 0.9753 -3.3814>, rot=<-0.0397 0.0814 -0.069 -0.9935>}
 pinky_01_r {parent=pinky_metacarpal_r, loc=<-7.1709 0.4831 0.1094>, rot=<-0.0654 -0.0035 0.2242 -0.9723>}
 pinky_02_r {parent=pinky_01_r, loc=<-2.8274 -0.0893 0.1043>, rot=<0.0037 0.0011 0.2854 -0.9584>}
 pinky_03_r {parent=pinky_02_r, loc=<-2.2046 -0.0307 0.0078>, rot=<-0.0002 0.0098 0.1445 -0.9895>}
 deform_wrist_dn_base_r {parent=hand_r, loc=<-0.0001 1.5328 0>, rot=<0 0 0 -1>}
 deform_wrist_dn_r {parent=deform_wrist_dn_base_r, loc=<0 0 0>, rot=<0 0 0 1>}
 deform_wrist_up_base_r {parent=hand_r, loc=<-0.0002 -2.3007 -0.0001>, rot=<0 0 0 -1>}
 deform_wrist_up_r {parent=deform_wrist_up_base_r, loc=<0 0 0>, rot=<0 0 0 1>}
 lowerarm_twist_02_r {parent=lowerarm_r, loc=<-10.6707 0 0>, rot=<0 0 0 -1>}
 lowerarm_twist_01_r {parent=lowerarm_r, loc=<-18.316 0 0>, rot=<0 0 0 -1>}
 deform_elbow_in_base_r {parent=lowerarm_r, loc=<-0.1834 -3.6062 0.2572>, rot=<0 0 0 -1>}
 deform_elbow_in_r {parent=deform_elbow_in_base_r, loc=<0 0 0>, rot=<0 0 0 1>}
 deform_elbow_out_base_r {parent=lowerarm_r, loc=<0 4 -0.0001>, rot=<0 0 0 -1>}
 deform_elbow_out_r {parent=deform_elbow_out_base_r, loc=<0 0 0>, rot=<0 0 0 1>}
 upperarm_twist_01_r {parent=upperarm_r, loc=<-7.2135 0.1134 0>, rot=<0 0 0 -1>}
 deform_upperarm_bk_base_r {parent=upperarm_twist_01_r, loc=<6.4053 4.3746 -0.314>, rot=<0 0 0 -1>}
 deform_upperarm_bk_r {parent=deform_upperarm_bk_base_r, loc=<0 0 0>, rot=<0 0 0 1>}
 deform_upperarm_fr_base_r {parent=upperarm_twist_01_r, loc=<5.6312 -5.092 -0.3354>, rot=<0 0 0 -1>}
 deform_upperarm_fr_r {parent=deform_upperarm_fr_base_r, loc=<0 0 0>, rot=<0 0 0 1>}
 deform_upperarm_up_base_r {parent=upperarm_twist_01_r, loc=<8.891 -0.4825 -3.9517>, rot=<0 0 0 -1>}
 deform_upperarm_up_r {parent=deform_upperarm_up_base_r, loc=<0 0 0>, rot=<0 0 0 1>}
 upperarm_twist_02_r {parent=upperarm_r, loc=<-19.7998 0.1134 0>, rot=<0 0 0 -1>}
 attach_backpack {parent=spine_05, loc=<-0.123 -13.473 0>, rot=<-0.0093 -0.707 0.0093 -0.707>}
 Backpack_BR {parent=attach_backpack, type=Null, loc=<0 0 0>, rot=<0 0 0 1>}
 attach_cape {parent=spine_05, loc=<9.017 -11.714 0>, rot=<0.0616 -0.7044 -0.0616 -0.7044>}
 cape {parent=attach_cape, type=Null, loc=<0 0 0>, rot=<0 0 0 1>}
 bustier_clav_l {parent=spine_05, loc=<5.4831 -0.1053 -3.2946>, rot=<0 0 0 1>}
 bustier_upperarm_l {parent=bustier_clav_l, loc=<-0.9312 -1.7869 -9.7424>, rot=<0.0385 -0.9118 -0.0238 -0.4081>}
 bustier_clav_r {parent=spine_05, loc=<5.4831 -0.1053 3.2946>, rot=<0 0 0 1>}
 bustier_upperarm_r {parent=bustier_clav_r, loc=<-0.9312 -1.7869 9.7424>, rot=<0.9118 0.0385 -0.4081 0.0238>}
 deform_pec_aim_base_l {parent=spine_04, loc=<3.2456 9.3309 -1.2237>, rot=<0.0603 -0.4033 0.3318 -0.8507>}
 deform_pec_aim_l {parent=deform_pec_aim_base_l, loc=<0 0 0>, rot=<0 0 0 1>}
 deform_pec_base_l {parent=deform_pec_aim_l, loc=<9.3665 0 0>, rot=<0 0 0 1>}
 deform_pec_l {parent=deform_pec_base_l, loc=<0 0 0>, rot=<0 0 0 1>}
 deform_pec_aim_base_r {parent=spine_04, loc=<3.2452 9.3309 1.2237>, rot=<-0.4033 -0.0603 0.8507 0.3318>}
 deform_pec_aim_r {parent=deform_pec_aim_base_r, loc=<0 0 0>, rot=<0 0 0 1>}
 deform_pec_base_r {parent=deform_pec_aim_r, loc=<-9.3667 -0.0003 -0.0004>, rot=<0 0 0 1>}
 deform_pec_r {parent=deform_pec_base_r, loc=<0 0 0>, rot=<0 0 0 1>}
 thigh_l {parent=pelvis, loc=<-1.0855 -0.2109 -8.7053>, rot=<0.0009 0.0355 0.0155 -0.9993>}
 calf_l {parent=thigh_l, loc=<-42.5284 -0.0019 -0.7297>, rot=<0.0006 -0.0161 -0.0391 -0.9991>}
 foot_l {parent=calf_l, loc=<-46.7401 -2.5014 -0.592>, rot=<0.0427 -0.0238 0.0353 -0.9982>}
 ball_l {parent=foot_l, loc=<-3.5658 10.8066 -0.1544>, rot=<0.0001 -0.0001 -0.7063 -0.7079>}
 calf_twist_02_l {parent=calf_l, loc=<-19.1546 -1.7943 -0.2638>, rot=<0.0237 0.0095 -0.0035 -0.9997>}
 calf_twist_01_l {parent=calf_l, loc=<-35.8483 -1.9197 -0.5794>, rot=<0.0237 0.0095 -0.0035 -0.9997>}
 deform_knee_in_base_l {parent=calf_l, loc=<0.8031 -5.0231 0.2456>, rot=<0 0 0 -1>}
 deform_knee_in_l {parent=deform_knee_in_base_l, loc=<0 0 0>, rot=<0 0 0 1>}
 deform_knee_out_base_l {parent=calf_l, loc=<0 4 0>, rot=<0 0 0 -1>}
 deform_knee_out_l {parent=deform_knee_out_base_l, loc=<0 0 0>, rot=<0 0 0 1>}
 thigh_twist_01_l {parent=thigh_l, loc=<-27.952 0 0>, rot=<0.0013 0 0 -1>}
 deform_groin_base_l {parent=thigh_l, loc=<0.4801 5.6776 -0.2486>, rot=<0 0 0 -1>}
 deform_groin_l {parent=deform_groin_base_l, loc=<0 0 0>, rot=<0 0 0 1>}
 thigh_r {parent=pelvis, loc=<-1.0855 -0.2109 8.7053>, rot=<0.0355 -0.0009 0.9993 0.0155>}
 calf_r {parent=thigh_r, loc=<42.5284 0.0019 0.7297>, rot=<0.0006 -0.0161 -0.0391 -0.9991>}
 foot_r {parent=calf_r, loc=<46.7401 2.5014 0.592>, rot=<0.0427 -0.0238 0.0353 -0.9982>}
 ball_r {parent=foot_r, loc=<3.5658 -10.8066 0.1544>, rot=<0.0001 -0.0001 -0.7063 -0.7079>}
 calf_twist_02_r {parent=calf_r, loc=<19.1546 1.7943 0.2638>, rot=<0.0237 0.0095 -0.0035 -0.9997>}
 calf_twist_01_r {parent=calf_r, loc=<35.8483 1.9197 0.5794>, rot=<0.0237 0.0095 -0.0035 -0.9997>}
 deform_knee_in_base_r {parent=calf_r, loc=<-0.8031 5.0231 -0.2457>, rot=<0 0 0 -1>}
 deform_knee_in_r {parent=deform_knee_in_base_r, loc=<0 0 0>, rot=<0 0 0 1>}
 deform_knee_out_base_r {parent=calf_r, loc=<0 -4 0>, rot=<0 0 0 -1>}
 deform_knee_out_r {parent=deform_knee_out_base_r, loc=<0 0 0>, rot=<0 0 0 1>}
 thigh_twist_01_r {parent=thigh_r, loc=<27.952 0 0>, rot=<0.0013 0 0 -1>}
 deform_groin_base_r {parent=thigh_r, loc=<-0.4801 -5.6776 0.2486>, rot=<0 0 0 -1>}
 deform_groin_r {parent=deform_groin_base_r, loc=<0 0 0>, rot=<0 0 0 1>}
 deform_glute_base_l {parent=pelvis, loc=<-7.3466 -9.0213 -9.1245>, rot=<0.0172 0.031 -0.4477 -0.8935>}
 deform_glute_l {parent=deform_glute_base_l, loc=<0 0 0>, rot=<0 0 0 1>}
 deform_glute_base_r {parent=pelvis, loc=<-7.3464 -9.0213 9.1245>, rot=<0.031 -0.0172 0.8935 -0.4477>}
 deform_glute_r {parent=deform_glute_base_r, loc=<0 0 0>, rot=<0 0 0 1>}
]
variables [
 Debug Bones {type="TArray<FRigElementKey>", default=((Type=Bone,Name="calf_l"),(Type=Bone,Name="calf_r"))}
 Debug Controls On {type=bool, default=False}
 Debug Ctrls {type="TArray<FRigElementKey>", default=(())}
]
